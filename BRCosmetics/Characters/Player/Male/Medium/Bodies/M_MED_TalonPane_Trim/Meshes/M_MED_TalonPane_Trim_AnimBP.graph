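# Animation Blueprint for the TalonPane Trim male body cosmetic, adding secondary motion on top of the parent mesh pose.
# The copied pose runs through a Control Rig, whose On/Off_Flail input is set from the TalonPaneFlail cosmetic meta tag.
# Weighted constraints then position both zipper base bones, and spring-based AnimDynamics swings the zippers and straps.
# The event graph updates a cape gravity override through a deprecated function, but no anim node reads it.
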AnimBlueprint /BRCosmetics/Characters/Player/Male/Medium/Bodies/M_MED_TalonPane_Trim/Meshes/M_MED_TalonPane_Trim_AnimBP
target_skeleton=&/M_MED_TalonPane_Trim_Skeleton
graphs [
 AnimGraph {nodes [
   n1 = Control Rig!(Source=Copy Pose From Mesh() {bUseAttachedParent=true}, Alpha, On/Off_Flail:double=Get On/Off_Flail() -> (On/Off_Flail:double)) -> (Pose) {
    alphaBoolBlend {blendOption=Linear}
    controlRigAssetReference {blueprintRigClass=&/M_MED_TalonPane_Trim_CtrlRig.M_MED_TalonPane_Trim_CtrlRig_C}
   }
   n2 = "Constraint - deform_zipper_base_r (PPPPP)"!(
    ComponentPose=Local To Component(LocalPose=@n1)
    Alpha=1
    ConstraintWeights_0=0.6
    ConstraintWeights_1:float=0.3
    ConstraintWeights_2:float=0.1
    ConstraintWeights_3:float=0
    ConstraintWeights_4:float=0
   ) -> (Pose) {
    alphaBoolBlend {blendOption=Linear}
    boneToModify {boneName=deform_zipper_base_r}
    constraintSetup [
     {targetBone {boneName=thigh_r}}
     {targetBone {boneName=calf_r}}
     {targetBone {boneName=deform_knee_out_r}}
     {targetBone {boneName=pelvis}}
     {targetBone {boneName=calf_twist_02_r}}
    ]
    constraintWeights [1,1,1,1,1]
   }
   n3 = "Constraint - deform_zipper_base_l (PPPPP)"!(
    ComponentPose=@n2
    Alpha=1
    ConstraintWeights_0=0.6
    ConstraintWeights_1:float=0.3
    ConstraintWeights_2:float=0.1
    ConstraintWeights_3:float=0
    ConstraintWeights_4:float=0
   ) -> (Pose) {
    alphaBoolBlend {blendOption=Linear}
    boneToModify {boneName=deform_zipper_base_l}
    constraintSetup [
     {targetBone {boneName=thigh_l}}
     {targetBone {boneName=calf_l}}
     {targetBone {boneName=deform_knee_out_l}}
     {targetBone {boneName=pelvis}}
     {targetBone {boneName=calf_twist_02_l}}
    ]
    constraintWeights [1,1,1,1,1]
   }
   n4 = Anim Dynamics - Bone: deform_zipper_l!(ComponentPose=@n3, Alpha=1, GravityOverride:Vector="0,0,500", ExternalForce="0,0,0") -> (Pose) {
    angularDampingOverride=0.99
    angularSpringConstant=800
    bAngularSpring=true
    bOverrideAngularDamping=true
    bUseGravityOverride=true
    gravityScale=0.3
    simulationSpace=BoneRelative
    alphaBoolBlend {blendOption=Linear}
    boundBone {boneName=deform_zipper_l}
    componentAppliedLinearAccClamp=<2000 2000 1600>
    componentLinearAccScale=<0.45 0.45 0.45>
    componentLinearVelScale=<0.55 0.55 0.55>
    gravityOverride {z=500}
    physicsBodyDefinitions [{
      collisionType=CustomSphere
      sphereCollisionRadius=2
      boundBone {boneName=deform_zipper_l}
      boxExtents=<20 10 10>
      constraintSetup {twistAxis=AxisY, angularLimitsMax=<25 5 50>, angularLimitsMin=<-5 -5 -10>, angularTarget=<0 0 0>}
      localJointOffset=<-5 -1 0.1>
     }]
    relativeSpaceBone {boneName=root}
   }
   n5 = Anim Dynamics - Bone: deform_zipper_r!(ComponentPose=@n4, Alpha=1, GravityOverride:Vector="0,0,500", ExternalForce="0,0,0") -> (Pose) {
    angularDampingOverride=0.99
    angularSpringConstant=800
    bAngularSpring=true
    bOverrideAngularDamping=true
    bUseGravityOverride=true
    gravityScale=0.3
    simulationSpace=BoneRelative
    alphaBoolBlend {blendOption=Linear}
    boundBone {boneName=deform_zipper_r}
    componentAppliedLinearAccClamp=<2000 2000 1600>
    componentLinearAccScale=<0.45 0.45 0.45>
    componentLinearVelScale=<0.55 0.55 0.55>
    gravityOverride {z=500}
    physicsBodyDefinitions [{
      collisionType=CustomSphere
      sphereCollisionRadius=2
      boundBone {boneName=deform_zipper_r}
      boxExtents=<20 10 10>
      constraintSetup {twistAxis=AxisY, angularLimitsMax=<25 5 10>, angularLimitsMin=<-5 -5 -50>, angularTarget=<0 0 0>}
      localJointOffset=<-5 -1 0.1>
     }]
    relativeSpaceBone {boneName=root}
   }
   n6 = Anim Dynamics - Bone: fix_strap_l!(ComponentPose=@n5, Alpha=1, GravityOverride:Vector="0,0,500", ExternalForce="0,0,0") -> (Pose) {
    angularDampingOverride=0.99
    angularSpringConstant=800
    bAngularSpring=true
    bOverrideAngularDamping=true
    bUseGravityOverride=true
    gravityScale=0.3
    simulationSpace=BoneRelative
    alphaBoolBlend {blendOption=Linear}
    boundBone {boneName=fix_strap_l}
    componentAppliedLinearAccClamp=<2000 2000 1600>
    componentLinearAccScale=<0.45 0.45 0.45>
    componentLinearVelScale=<0.55 0.55 0.55>
    gravityOverride {z=500}
    physicsBodyDefinitions [{
      collisionType=CustomSphere
      sphereCollisionRadius=2
      boundBone {boneName=fix_strap_l}
      boxExtents=<20 20 10>
      constraintSetup {twistAxis=AxisZ, angularLimitsMax=<80 5 10>, angularLimitsMin=<-5 -5 -10>, angularTarget=<0 0 0>}
      localJointOffset=<0.1 2 -5>
     }]
    relativeSpaceBone {boneName=root}
   }
   n7 = Anim Dynamics - Bone: fix_strap_r!(ComponentPose=@n6, Alpha=1, GravityOverride:Vector="0,0,500", ExternalForce="0,0,0") -> (Pose) {
    angularDampingOverride=0.99
    angularSpringConstant=800
    bAngularSpring=true
    bOverrideAngularDamping=true
    bUseGravityOverride=true
    gravityScale=0.3
    simulationSpace=BoneRelative
    alphaBoolBlend {blendOption=Linear}
    boundBone {boneName=fix_strap_r}
    componentAppliedLinearAccClamp=<2000 2000 1600>
    componentLinearAccScale=<0.45 0.45 0.45>
    componentLinearVelScale=<0.55 0.55 0.55>
    gravityOverride {z=500}
    physicsBodyDefinitions [{
      collisionType=CustomSphere
      sphereCollisionRadius=2
      boundBone {boneName=fix_strap_r}
      boxExtents=<20 20 10>
      constraintSetup {twistAxis=AxisZ, angularLimitsMax=<80 5 10>, angularLimitsMin=<-5 -5 -10>, angularTarget=<0 0 0>}
      localJointOffset=<0.1 2 -5>
     }]
    relativeSpaceBone {boneName=root}
   }
   Output Pose(Result=Component To Local(ComponentPose=@n7)) {layerGroup=None}
  ]}
 EventGraph {nodes [
   Event BlueprintUpdateAnimation()
   > Sequence()
   > n1 = Gravity Override State Function DEPRECATED(CharacterPartAnimInstance=Self-Reference(), GameTimeInSeconds=GetGameTimeInSeconds(), GravityOverrideParams=Get Gravity Override Params())
   > Set Gravity Override Cape(Gravity Override Cape:Vector=@n1) -> (Output_Get:Vector)
   n2 = Event OnCharacterCustomizationLoaded()
   n3 = Branch(
    execute=[@n2.then, @n4.then]
    Condition=CharacterItemDefinitionHasAnyMetaTag(Tags=Make Array("[0]"=(TagName="Cosmetics.MetaTags.TalonPaneFlail")))
   )
   > Set On/Off_Flail(On/Off_Flail:double=1) -> (Output_Get:double)
   Set On/Off_Flail(execute=@n3.else, On/Off_Flail=0)
   n4 = Event BlueprintInitializeAnimation()
  ]}
]
variables [
 Gravity Override Cape {type=Vector}
 Gravity Override Params {
  type=GravityOverrideParamsStruct
  default {
   onGround_Falling {jointName=pelvis, gravityOverride=<-980 0 0>}
   onGround_Moving {jointName=pelvis, gravityOverride=<-980 200 0>}
   onGround_Standing {jointName=pelvis, gravityOverride=<-980 0 0>}
   skydive_FloatingDown {gravityOverride=<0 -700 -300>, windAmplitude=<200 -500 200>, windOffset=<0 0 0>}
   skydive_GoingUp {gravityOverride=<-150 200 1500>, windAmplitude=<-500 900 -450>, windFrequency=<1.5 18 7>}
  }
 }
 On/Off_Flail {type=number}
]
